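# Material function for the SuperNovaTaro outfit that adds an animated black hole and galaxy effect, overriding only Base Color and Emissive of the input attributes.
# Pre-skinned positions drive a chest event horizon of scrolling radial noise, plus bump-offset waves that warp the body diffuse.
# Masks composite flowing star layers, tinted clouds and a distorted virtual-plane nebula with fresnel rim.
# Time-of-day functions adjust glow; UseCustomizedGalaxy switches an alternate look.
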MaterialFunction /Game/Characters/Player/Female/Medium/Bodies/F_MED_SuperNovaTaro/Materials/MF_BlackHole_FX_2023
graph {nodes [
  n1 = MaterialFunctionCall(Float3=NamedRerouteUsage()) {materialFunction=&BreakOutFloat3Components}
  n2 = ScalarParameter() {defaultValue=140, group=BlackHole, parameterName=StartHeight}
  n3 = MaterialFunctionCall(Z=Subtract(A=@n1.B, B=@n2)) {materialFunction=&MakeFloat3}
  n4 = ScalarParameter() {group=BlackHole, parameterName=EndHeight}
  n5 = Divide(A=DotProduct(A=Constant3Vector() {constant=<rgba 0 0 -1 0>}, B=Subtract(B=@n3.Result)), B=Abs(None=Subtract(A=@n2, B=@n4) {constB=0}))
  Abs(None=@n5)
  n6 = NamedRerouteUsage()
  ComponentMask(None=@n6) {r=true}
  n7 = GetMaterialAttributes(None=NamedRerouteUsage()) {attributeGetTypes [#1, #2, #3]}
  FunctionOutput(None=SetMaterialAttributes!(MaterialAttributes=@n7.MaterialAttributes, EmissiveColor=NamedRerouteUsage(), BaseColor=NamedRerouteUsage()) {attributeSetTypes [#1, #2]})
  n8 = Time()
  Multiply(A=@n8) {constB=0.05}
  n9 = MaterialFunctionCall!(Coordinate=MaterialFunctionCall(UVs=NamedRerouteUsage(), Texture Scale=Constant() {r=5}) {materialFunction=&ScaleUVsByCenter}.UVs) {materialFunction=/Game/Effects/Niagara/MasterMaterials/Functions/UVs/PolarCoordinates}
  n10 = Multiply(A=Time()) {constB=0.15}
  n11 = LinearInterpolate(
   A=Subtract(A=@n10) {constB=0.25}
   B=@n10
   Alpha=SphereMask(A=Constant() {r=0.5}, B=TextureCoordinate()) {attenuationRadius=0.5, hardnessPercent=0}
  )
  n12 = TextureCoordinate()
  n13 = TextureSample!(
   UVs=Add(
    A=ComponentMask(None=@n9.Result) {g=true, r=true}
    B=MaterialFunctionCall(X=@n11, Y=@n10) {materialFunction=&MakeFloat2}.Result
   )
   Tex
   "DDX(UVs)"=DDX(Value=@n12)
   "DDY(UVs)"=DDY(Value=@n12)
   Apply View MipBias
  ) {mipValueMode=TMVM_Derivative, texture=/Game/Global/Textures/Default/Noise/T_Tiling_Noise_01}
  n14 = Multiply(A=LinearInterpolate(Alpha=Power(Base=@n13.RGB)) {constA=0.5}, B=NamedRerouteUsage())
  Multiply(A=Power(Base=@n14)) {constB=2}
  n15 = Multiply(A=NamedRerouteUsage()) {constB=10}
  Multiply(A=OneMinus(None=Saturate(None=@n15)), B=NamedRerouteUsage())
  n16 = OneMinus(None=Distance(A=NamedRerouteUsage(), B=Constant() {r=0.5}))
  n17 = ScalarParameter() {group=BlackHole, parameterName=BHSizeGain}
  NamedRerouteDeclaration(None=Add(A=Saturate(None=@n16), B=@n17)) {name=GradientRadialLinear, nodeColor=<rgba 1 0.75 0 1>}
  NamedRerouteDeclaration(None=FunctionInput() {inputName=MaterialInput, inputType=FunctionInput_MaterialAttributes}) {name=MaterialInput, nodeColor=<rgba 0 0.7891 1 1>}
  n18 = Multiply(
   A=VectorParameter() {group=BlackHole, parameterName=StarTint01, defaultValue=<rgba 1 0 1 1>}.RGB
   B=ScalarParameter() {defaultValue=50, group=BlackHole, parameterName=StarBrightness01}
  ) {constB=3}
  n19 = Divide(A=@n18, B=StaticSwitchParameter(True=Constant() {r=1.6}, False=Constant() {r=1.2}) {group=BlackHole, parameterName=UseCustomizedGalaxy}) {constB=1.2}
  n20 = VectorParameter() {group=BlackHole, parameterName=DistorSpeed, channelNames {a=SpeedY1, b=SpeedX1, g=SpeedY0, r=SpeedX0}, defaultValue=<rgba 0 0.3 0 0>}
  n21 = Add(
   A=TextureCoordinate()
   B=Multiply(
    A=MaterialFunctionCall(X=@n20.SpeedX0, Y=@n20.SpeedX0) {materialFunction=&MakeFloat2}.Result
    B=NamedRerouteUsage()
   )
  )
  n22 = LinearInterpolate(A=Constant2Vector() {g=1, r=-0.8}, B=Constant2Vector() {g=1, r=0.8}, Alpha=NamedRerouteUsage())
  n23 = Constant2Vector() {g=1, r=-1}
  n24 = LinearInterpolate(A=LinearInterpolate(A=Multiply(A=@n22, B=@n23), B=@n22, Alpha=NamedRerouteUsage()), B=Constant2Vector() {g=1}, Alpha=NamedRerouteUsage())
  n25 = VectorParameter() {group=BlackHole, parameterName=ArmFlowDirection, channelNames {a=V02, b=U02, g=V01, r=U01}, defaultValue=<rgba 1 0 -1 0>}
  n26 = LinearInterpolate(
   A=MaterialFunctionCall(X=@n25.U01, Y=@n25.U01) {materialFunction=&MakeFloat2}.Result
   B=MaterialFunctionCall(X=@n25.U02, Y=@n25.U02) {materialFunction=&MakeFloat2}.Result
   Alpha=NamedRerouteUsage()
  )
  n27 = NamedRerouteUsage()
  n28 = NamedRerouteDeclaration(None=LinearInterpolate(A=LinearInterpolate(A=@n24, B=@n26, Alpha=@n27), B=Constant2Vector() {r=1}, Alpha=NamedRerouteUsage())) {name=MoveSpeed01, nodeColor=<rgba 1 0 0.3281 1>}
  n29 = Panner(
   Coordinate=Multiply(
    A=ScalarParameter() {defaultValue=13, group=BlackHole, parameterName=Black Hole Bright Star Tiling 01}
    B=NamedRerouteDeclaration(None=@n21) {name=DistortedUV, nodeColor=<rgba 0.4531 0 1 1>}
   )
   Speed=Multiply(A=@n28, B=ScalarParameter() {defaultValue=0.3, group=BlackHole, parameterName=StarMoveSpeed01}) {constB=0.2}
  ) {speedY=0.15}
  n30 = Multiply(A=@n29) {constB=4}
  n31 = ComponentMask(None=MaterialFunctionCall!(UVIndexInput=Floor(None=@n30)) {materialFunction=&/MF_RandomNoise}.Result) {r=true}
  n32 = TextureSample!(UVs=BumpOffset(Coordinate=@n29, Height=LinearInterpolate(Alpha=@n31) {constA=-1, constB=-10}), Tex=NamedRerouteUsage(), Apply View MipBias) {
   samplerSource=SSM_Wrap_WorldGroupSettings
   samplerType=SAMPLERTYPE_Masks
   texture=/Game/Characters/Player/Female/Medium/Bodies/F_MED_SuperNovaTaro/Textures/T_Stars_03
  }
  n33 = LinearInterpolate(Alpha=@n32.G) {constA=0.8}
  n34 = TextureCoordinate() {uTiling=2, vTiling=2}
  n35 = TextureSample(UVs=Panner(Coordinate=Multiply(A=@n34) {constB=2}, Speed=Multiply(A=@n28) {constB=0.5}) {speedY=0.2}, Tex=NamedRerouteUsage()) {
   samplerSource=SSM_Wrap_WorldGroupSettings
   samplerType=SAMPLERTYPE_Masks
   texture=&T_Stars_03
  }
  n36 = NamedRerouteDeclaration(None=@n35.G) {name=CachedTwinkle01, nodeColor=<rgba 0 1 0.4141 1>}
  n37 = Multiply(A=Power(Base=Multiply(A=@n32.R, B=@n33), Exp=ScalarParameter() {defaultValue=1.5, group=BlackHole, parameterName=StarPower01}), B=@n36)
  n38 = Constant3Vector()
  n39 = NamedRerouteUsage()
  n40 = Step(X=StaticComponentMaskParameter(None=NamedRerouteUsage()) {defaultG=true, group=BlackHole, parameterName=UseBlackHole}) {constY=0.3}
  n41 = Multiply(A=LinearInterpolate(A=LinearInterpolate(A=@n37, B=@n38, Alpha=@n39), B=@n38, Alpha=NamedRerouteUsage()), B=@n40)
  n42 = NamedRerouteUsage()
  n43 = LinearInterpolate(
   A=NamedRerouteUsage()
   B=MaterialFunctionCall(Color_Morning=@n18, Color_Day=@n18, Color_Evening=@n19, Color_Night=@n19) {materialFunction=&Color_Time_of_Day}.Result
   Alpha=Multiply(A=LinearInterpolate(A=@n41, Alpha=@n42) {constB=0}, B=NamedRerouteUsage())
  )
  n44 = Multiply(
   A=VectorParameter() {group=BlackHole, parameterName=StarTint02, defaultValue=<rgba 0.2109 0 1 1>}.RGB
   B=ScalarParameter() {defaultValue=50, group=BlackHole, parameterName=StarBrightness02}
  ) {constB=3}
  n45 = NamedRerouteUsage()
  n46 = ScalarParameter() {defaultValue=15, group=BlackHole, parameterName=Black Hole Bright Star Tiling 02}
  n47 = NamedRerouteUsage()
  n48 = ScalarParameter() {defaultValue=0.5, group=BlackHole, parameterName=StarMoveSpeed02}
  n49 = TextureSample(UVs=Panner(Coordinate=Multiply(A=@n46, B=@n47), Speed=Multiply(A=@n28, B=@n48)), Tex=NamedRerouteUsage()) {
   samplerSource=SSM_Wrap_WorldGroupSettings
   samplerType=SAMPLERTYPE_Masks
   texture=&T_Stars_03
  }
  n50 = Constant3Vector()
  n51 = LinearInterpolate(
   A=Power(Base=Multiply(A=@n36, B=@n49.B), Exp=ScalarParameter() {defaultValue=1.5, group=BlackHole, parameterName=StarPower02})
   B=@n50
   Alpha=NamedRerouteUsage()
  )
  n52 = NamedRerouteUsage()
  n53 = Multiply(
   A=Multiply(A=NamedRerouteUsage(), B=LinearInterpolate(A=@n44, Alpha=@n45) {constB=0})
   B=Multiply(A=@n40, B=LinearInterpolate(A=@n51, B=@n50, Alpha=@n52))
  )
  n54 = Divide(A=@n53, B=StaticSwitchParameter(True=Constant() {r=3}, False=Constant() {r=2}) {group=BlackHole, parameterName=UseCustomizedGalaxy})
  n55 = MaterialFunctionCall(Color_Morning=@n53, Color_Day=@n53, Color_Evening=@n54, Color_Night=@n54) {materialFunction=&Color_Time_of_Day}
  n56 = Subtract(A=Saturate(None=@n55.Result), B=Saturate(None=@n41))
  n57 = NamedRerouteDeclaration(None=Clamp(
    None=LinearInterpolate(A=@n43, B=@n55.Result, Alpha=@n56)
    Max=ScalarParameter() {defaultValue=25, group=BlackHole, parameterName=EmissionClamp}
   )) {name=CachedStarOrg, nodeColor=<rgba 0.4297 0 1 1>}
  n58 = ScalarParameter() {defaultValue=1.2, group=BlackHole, parameterName=EmissionMultiIntensity}
  n59 = ScalarParameter() {defaultValue=0.3, group=BlackHole, parameterName=HeightLerpMin}
  n60 = ScalarParameter() {defaultValue=3, group=BlackHole, parameterName=HeightLerpMax}
  n61 = ComponentMask(None=NamedRerouteUsage()) {b=true}
  n62 = ScalarParameter() {defaultValue=60, group=BlackHole, parameterName=HeightLerpRange}
  n63 = Saturate(None=Abs(None=Divide(A=@n61, B=@n62) {constB=60}))
  n64 = LinearInterpolate(Alpha=Clamp(None=OneMinus(None=@n63)) {maxDefault=0.5}) {constB=2}
  n65 = OneMinus(None=Saturate(None=NamedRerouteUsage()))
  n66 = OneMinus(None=Saturate(None=NamedRerouteUsage()))
  n67 = Multiply(A=Multiply(
    A=NamedRerouteDeclaration(None=@n64) {name=DebugHeightLerp, nodeColor=<rgba 0.4766 1 0 1>}
    B=StaticSwitchParameter(True=@n65, False=@n66) {group=BlackHole, parameterName=UseCustomizedGalaxy}
   ))
  n68 = NamedRerouteDeclaration(None=Multiply(A=Multiply(A=@n57, B=@n58), B=LinearInterpolate(A=@n59, B=@n60, Alpha=@n67) {constA=0.7, constB=3})) {name=CachedBeforeSucking, nodeColor=<rgba 1 0 0.3047 1>}
  n69 = NamedRerouteUsage()
  n70 = StaticSwitchParameter(True=Constant(), False=OneMinus(None=NamedRerouteUsage())) {group=BlackHole, parameterName=UseCustomizedGalaxy}
  n71 = NamedRerouteUsage()
  n72 = NamedRerouteUsage()
  n73 = Step(X=StaticComponentMaskParameter(None=NamedRerouteUsage()) {defaultR=true, group=BlackHole, parameterName=MetalMaskChannel}) {constY=0.05}
  n74 = OneMinus(None=@n73)
  n75 = StaticSwitchParameter(
   True=@n72
   False=NamedRerouteDeclaration(None=StaticSwitchParameter(True=@n74, False=@n73) {defaultValue=true, group=BlackHole, parameterName=MetalMaskRevert}) {name=DebugMetal, nodeColor=<rgba 0 0.1094 1 1>}
  ) {group=BlackHole, parameterName=UseCustomizedGalaxy}
  NamedRerouteDeclaration(None=Multiply(A=LinearInterpolate(A=@n68, B=@n69, Alpha=@n70), B=LinearInterpolate(B=@n71, Alpha=@n75) {constA=1})) {name=CachedEmissive, nodeColor=<rgba 0.5 1 0 1>}
  n76 = PixelDepth()
  NamedRerouteDeclaration(None=Saturate(None=Divide(A=@n76.R) {constB=2000})) {name=CachedDepthlerp, nodeColor=<rgba 0 0.7187 1 1>}
  n77 = TextureSample(UVs=NamedRerouteUsage()) {
   samplerSource=SSM_Wrap_WorldGroupSettings
   samplerType=SAMPLERTYPE_Masks
   texture=/Game/Characters/Player/Female/Medium/Bodies/F_MED_SuperNovaTaro/Textures/T_SuperNovaTaro_Body_FX
  }
  n78 = Subtract(A=OneMinus(None=ComponentMask(None=@n77.RGB) {g=true}), B=Ceil(None=ComponentMask(None=@n77.RGB) {r=true}))
  NamedRerouteDeclaration(None=StaticSwitchParameter(True=Constant(), False=SmoothStep(Value=@n78) {constMin=0.2}) {group=BlackHole, parameterName=UseCustomizedGalaxy}) {name=CachedBodyPaintMask, nodeColor=<rgba 0.2422 1 0 1>}
  n79 = NamedRerouteUsage()
  n80 = NamedRerouteUsage()
  n81 = StaticSwitchParameter(
   True=LinearInterpolate(A=VectorParameter() {group=BlackHole, parameterName=SecondOptionalColor, defaultValue=<rgba 0.6136 0 1 1>}.RGB, B=@n79, Alpha=@n80)
   False=@n79
  ) {group=BlackHole, parameterName=UseLerpedPaintColor}
  n82 = ScalarParameter() {defaultValue=3, group=BlackHole, parameterName=PatternIntensity}
  n83 = Constant() {r=-30}
  n84 = Constant() {r=-140}
  n85 = Frac(None=Multiply(A=@n8) {constB=0.15})
  n86 = Multiply(A=Subtract(A=@n85) {constB=0.875}) {constB=8}
  n87 = LinearInterpolate(A=Saturate(None=Multiply(A=@n85) {constB=1.15}), B=Saturate(None=OneMinus(None=@n86)), Alpha=Step(X=@n85) {constY=0.875})
  n88 = Add(A=PreSkinnedPosition(), B=AppendVector(A=Constant2Vector(), B=LinearInterpolate(A=@n83, B=@n84, Alpha=@n87)))
  n89 = ComponentMask(None=VertexInterpolator(VS=Divide(A=@n88) {constB=100}).PS) {b=true}
  n90 = LinearInterpolate(A=NamedRerouteUsage(), B=Add(A=Multiply(A=@n81, B=@n82) {constB=2}, B=Constant()), Alpha=Multiply(A=@n80, B=Saturate(None=@n89)))
  n91 = Saturate(None=Multiply(A=NamedRerouteUsage()) {constB=10})
  n92 = LinearInterpolate(
   A=Multiply(A=NamedRerouteUsage(), B=ScalarParameter() {defaultValue=0.6, group=GalaxyScreenSpace, parameterName=GalaxyIntensity})
   B=NamedRerouteUsage()
   Alpha=Multiply(A=OneMinus(None=@n91), B=NamedRerouteUsage())
  )
  n93 = Multiply(A=@n92) {constB=0.3}
  n94 = StaticSwitchParameter(
   True=Add(
    A=@n90
    B=MaterialFunctionCall(Color_Morning=@n92, Color_Day=@n92, Color_Evening=@n93, Color_Night=@n93) {materialFunction=&Color_Time_of_Day}.Result
   )
   False=@n90
  ) {group=BlackHole, parameterName=UseCustomizedGalaxy}
  n95 = Divide(A=NamedRerouteUsage(), B=PreSkinnedLocalBounds().Extents)
  n96 = ComponentMask(None=@n95) {r=true}
  n97 = OneMinus(None=ComponentMask(None=@n95) {b=true})
  n98 = Divide(A=ComponentMask(None=@n95) {g=true}) {constB=20}
  n99 = Multiply(A=LinearInterpolate(
    A=MaterialFunctionCall(X=@n96, Y=@n97) {materialFunction=&MakeFloat2}.Result
    B=MaterialFunctionCall(X=@n98, Y=@n97) {materialFunction=&MakeFloat2}.Result
    Alpha=NamedRerouteUsage()
   )) {constB=3}
  n100 = Multiply(A=Constant2Vector() {g=0.5}, B=NamedRerouteUsage())
  n101 = TextureSample(UVs=Panner(Coordinate=Add(A=@n99, B=@n100), Speed=Constant2Vector() {g=0.1})) {
   samplerType=SAMPLERTYPE_LinearColor
   texture=/Game/Characters/Player/Female/Medium/Bodies/F_MED_SuperNovaTaro/Textures/T_CloudCLuster_LinearRGB
  }
  n102 = StaticComponentMaskParameter(None=NamedRerouteUsage()) {defaultG=true, group=BlackHole, parameterName=UseBlackHole}
  n103 = Multiply(
   A=@n101.RGB
   B=StaticSwitchParameter(True=Step(X=@n102) {constY=0.2}, False=Step(X=@n102) {constY=0.3}) {group=BlackHole, parameterName=UseCustomizedGalaxy}
  )
  n104 = Multiply(
   A=NamedRerouteUsage()
   B=Multiply(A=ComponentMask(None=@n103) {r=true}, B=VectorParameter() {group=BlackHole, parameterName=CloudTint01, defaultValue=<rgba 0 0.8219 1 1>}.RGB)
  )
  n105 = Multiply(
   A=NamedRerouteUsage()
   B=Multiply(A=VectorParameter() {group=BlackHole, parameterName=CloudTint02, defaultValue=<rgba 0 0.0345 1 1>}.RGB, B=ComponentMask(None=@n103) {g=true})
  )
  n106 = NamedRerouteUsage()
  n107 = Multiply(A=ComponentMask(None=@n103) {b=true}, B=VectorParameter() {group=BlackHole, parameterName=CloudTint03, defaultValue=<rgba 1 0 0.479 1>}.RGB)
  n108 = Multiply(A=Add(A=Add(A=@n104, B=@n105), B=Multiply(A=@n106, B=@n107)), B=OneMinus(None=NamedRerouteUsage()))
  n109 = LinearInterpolate(Alpha=NamedRerouteUsage()) {constA=0.45}
  n110 = Multiply(
   A=LinearInterpolate(A=Multiply(A=@n108, B=@n109), Alpha=NamedRerouteUsage()) {constB=0}
   B=ScalarParameter() {defaultValue=1, group=BlackHole, parameterName=Blackhole_CloudIntensity}
  )
  n111 = Multiply(A=@n110) {constB=0.5}
  n112 = MaterialFunctionCall(Color_Morning=@n110, Color_Day=@n110, Color_Evening=@n111, Color_Night=@n111) {materialFunction=&Color_Time_of_Day}
  NamedRerouteDeclaration(None=Add(A=NamedRerouteDeclaration(None=@n94) {name=CachePostfx, nodeColor=<rgba 0.2891 0 1 1>}, B=@n112.Result)) {name=CachedEmissiveOUT, nodeColor=<rgba 0.0078 1 0 1>}
  NamedRerouteDeclaration(None=FunctionInput() {inputName=InputWorldPos, sortPriority=1}) {name=SkinLocation, nodeColor=<rgba 0 1 0.625 1>}
  n113 = Abs(None=@n1.R)
  n114 = Constant3Vector() {constant=<rgba 60 -6 88 0>}
  n115 = Constant3Vector() {constant=<rgba -1 -0.15 0.75 0>}
  n116 = DotProduct(
   A=Subtract(
    A=MaterialFunctionCall(X=@n113, Y=@n1.G, Z=@n1.G) {materialFunction=&MakeFloat3}.Result
    B=@n114
   )
   B=Normalize(VectorInput=@n115)
  )
  n117 = Distance(A=@n114, B=Constant3Vector() {constant=<rgba 0 8 150 0>})
  NamedRerouteDeclaration(None=Saturate(None=Divide(A=@n116, B=@n117))) {name=CachedArmGradientRaw, nodeColor=<rgba 0 0.8828 1 1>}
  NamedRerouteDeclaration(None=@n7.BaseColor) {name=CachedBaseColorIn, nodeColor=<rgba 1 0 0.8906 1>}
  n118 = NamedRerouteUsage()
  n119 = Step(X=@n118) {constY=0.6}
  n120 = OneMinus(None=Saturate(None=@n119))
  n121 = Step(X=StaticComponentMaskParameter(None=NamedRerouteUsage()) {defaultR=true, group=BlackHole, parameterName=MetalMaskChannel}) {constY=0.05}
  n122 = NamedRerouteDeclaration(None=StaticSwitchParameter(True=OneMinus(None=@n121), False=@n121) {defaultValue=true, group=BlackHole, parameterName=MetalMaskRevert}) {name=DebugMetalMask_BlackVersion, nodeColor=<rgba 0.1484 0 1 1>}
  n123 = ComponentMask(None=VertexInterpolator(VS=PreSkinnedPosition()).PS) {g=true, r=true}
  n124 = ScalarParameter() {defaultValue=0.8, group=BlackHole, parameterName=CharacterGradient_RumpFalloffHardness, sliderMax=1, sortPriority=1000}
  n125 = OneMinus(None=SphereMask(
    A=@n123
    B=Constant2Vector()
    Radius=ScalarParameter() {defaultValue=30, group=BlackHole, parameterName=CharacterGradient_RumpRadius, sortPriority=1000}
    Hardness=@n124
   ) {attenuationRadius=30, hardnessPercent=80})
  NamedRerouteDeclaration(None=Multiply(A=StaticSwitchParameter(True=@n120, False=@n122) {group=BlackHole, parameterName=UseCustomizedGalaxy}, B=@n125)) {name=CachedArmMask, nodeColor=<rgba 0.4062 0 1 1>}
  n126 = TextureObjectParameter() {
   group=BlackHole
   parameterName=StarForBlackHole
   samplerType=SAMPLERTYPE_Masks
   texture=&T_Stars_03
  }
  NamedRerouteDeclaration(None=@n126) {name=CachedStarTexture, nodeColor=<rgba 0.4062 1 0 1>}
  n127 = Subtract(B=@n115) {constA=0}
  n128 = Step(X=@n1.R)
  NamedRerouteDeclaration(None=Transform(None=LinearInterpolate(A=@n127, B=@n115, Alpha=@n128)) {transformSourceType=TRANSFORMSOURCE_World, transformType=TRANSFORM_Tangent}) {name=DebugVector, nodeColor=<rgba 0.4766 1 0 1>}
  n129 = ScalarParameter() {defaultValue=0.8, group=BlackHole, parameterName=CharacterGradient_RumpFalloffHardnesshand, sliderMax=1, sortPriority=1000}
  n130 = SphereMask(
   A=@n123
   B=Constant2Vector()
   Radius=ScalarParameter() {defaultValue=30, group=BlackHole, parameterName=CharacterGradient_RumpRadiushand, sortPriority=1000}
   Hardness=@n129
  ) {attenuationRadius=30, hardnessPercent=80}
  NamedRerouteDeclaration(None=Multiply(A=OneMinus(None=@n130), B=NamedRerouteUsage())) {name=CachedLeftHandMask, nodeColor=<rgba 0 1 0.3438 1>}
  n131 = Subtract(A=NamedRerouteUsage(), B=VectorParameter() {group=BlackHole, parameterName=BlackHoleCenter, defaultValue=<rgba 0 0 0 1>}.RGB)
  n132 = Normalize(VectorInput=ComponentMask(None=@n131) {b=true, g=true, r=true})
  NamedRerouteDeclaration(None=Step(X=ComponentMask(None=@n132) {r=true})) {name=RMask, nodeColor=<rgba 0 1 0.4375 1>}
  NamedRerouteDeclaration(None=Step(X=ComponentMask(None=@n132) {g=true})) {name=GMask, nodeColor=<rgba 1 0.9844 0 1>}
  NamedRerouteDeclaration(None=Step(X=ComponentMask(None=@n132) {b=true})) {name=BMask, nodeColor=<rgba 0 0.9531 1 1>}
  n133 = Saturate(None=@n5)
  NamedRerouteDeclaration(None=OneMinus(None=@n133)) {name=CachedlegRange, nodeColor=<rgba 0.3359 0 1 1>}
  n134 = NamedRerouteUsage()
  n135 = OneMinus(None=Ceil(None=StaticComponentMaskParameter(None=@n6) {defaultR=true, group=BlackHole, parameterName=UseBlackHoleSolidPaint}))
  n136 = Multiply(A=NamedRerouteUsage(), B=@n135)
  n137 = NamedRerouteUsage()
  n138 = NamedRerouteUsage()
  n139 = ScalarParameter() {defaultValue=0.3, group=BlackHole, parameterName=BHValueStep}
  n140 = Divide(A=ScalarParameter() {defaultValue=0.5, group=BlackHole, parameterName=BHSmothness}) {constB=10}
  n141 = MaterialFunctionCall!(Vector or UVs=NamedRerouteUsage()) {materialFunction=/FortUILibrary/MaterialFunctions/Basics/MF_UI_CheaperVectorToRadial}
  n142 = Multiply(A=ComponentMask(None=@n141.Radial Coordinates) {r=true}, B=ScalarParameter() {defaultValue=1, group=BlackHole, parameterName=NoiseTilingX}) {constB=2}
  n143 = Multiply(A=ComponentMask(None=@n141.Radial Coordinates) {g=true}, B=ScalarParameter() {defaultValue=5, group=BlackHole, parameterName=NoiseTilingY}) {constB=2}
  n144 = ScalarParameter() {defaultValue=-0.2, group=BlackHole, parameterName=ScrollSpeedX}
  n145 = ScalarParameter() {defaultValue=0.5, group=BlackHole, parameterName=ScrollSpeedY}
  n146 = TextureSample(UVs=Panner(
    Coordinate=MaterialFunctionCall(R=@n142, G=@n143) {materialFunction=&AppendMany}.RG
    Speed=AppendVector(A=@n144, B=@n145)
   ) {speedX=1, speedY=1}) {
   samplerType=SAMPLERTYPE_LinearGrayscale
   texture=/Game/Characters/Player/Female/Medium/Bodies/F_MED_SuperNovaTaro/Textures/LowResBlurredNoise
  }
  n147 = NamedRerouteUsage()
  n148 = Multiply(A=@n146.R, B=Add(A=@n147, B=ScalarParameter() {defaultValue=0.3, group=BlackHole, parameterName=EventHorizonOut}) {constB=0})
  n149 = Add(A=@n147, B=ScalarParameter() {group=BlackHole, parameterName=EventHorizonIn})
  n150 = NamedRerouteDeclaration(None=SmoothStep(Min=Subtract(A=@n139, B=@n140), Max=Add(A=@n139, B=@n140), Value=Add(A=@n148, B=@n149)) {constMax=0.35, constMin=0.25}) {name=CachedHoleEmissionMask, nodeColor=<rgba 0.1953 0 1 1>}
  n151 = VectorParameter() {group=BlackHole, parameterName=FXColor, defaultValue=<rgba 1 1 1 1>}
  n152 = Multiply(
   A=Multiply(A=Multiply(A=@n138, B=@n150), B=NamedRerouteDeclaration(None=@n151.RGB) {name=CachedEmissiveColor01, nodeColor=<rgba 0 1 0.2031 1>})
   B=@n135
  )
  n153 = NamedRerouteUsage()
  NamedRerouteDeclaration(None=StaticSwitchParameter(True=LinearInterpolate(A=@n134, B=@n136, Alpha=@n137), False=Add(A=@n152, B=@n153)) {group=BlackHole, parameterName=LerpEmission}) {name=ChestFx, nodeColor=<rgba 0 0.0625 1 1>}
  n154 = Divide(A=@n131, B=ScalarParameter() {defaultValue=2, group=BlackHole, parameterName=Dist01}) {constB=20}
  n155 = OneMinus(None=Saturate(None=Abs(None=@n154)))
  n156 = Clamp(None=Length(None=ComponentMask(None=@n154) {b=true}) {is=class MaterialExpressionLength})
  n157 = Divide(A=@n131, B=ScalarParameter() {defaultValue=2, group=BlackHole, parameterName=Dist02})
  n158 = OneMinus(None=Saturate(None=Abs(None=@n157)))
  n159 = ComponentMask(None=@n158) {r=true}
  n160 = Multiply(
   A=Multiply(A=ComponentMask(None=@n155) {r=true}, B=Power(Base=@n156))
   B=Clamp(None=Multiply(A=@n159) {constB=2}) {maxDefault=2, minDefault=1}
  )
  n161 = Saturate(None=Multiply(A=SmoothStep(Value=@n160) {constMax=5}) {constB=5})
  n162 = Multiply(A=ComponentMask(None=@n158) {g=true}, B=OneMinus(None=NamedRerouteUsage()))
  NamedRerouteDeclaration(None=Add(
    A=NamedRerouteDeclaration(None=@n161) {name=XSeam, nodeColor=<rgba 0.8281 0 1 1>}
    B=NamedRerouteDeclaration(None=@n162) {name=YSeam, nodeColor=<rgba 0 0.0156 1 1>}
   )) {name=SeamMask, nodeColor=<rgba 0.0312 1 0 1>}
  n163 = NamedRerouteUsage()
  n164 = Divide(A=@n163) {constB=10}
  n165 = ScalarParameter() {defaultValue=1, group=GalaxyScreenSpace, parameterName=Galaxy_GlobalBrightness}
  n166 = NamedRerouteUsage()
  n167 = ComponentMask(None=VectorParameter() {group=GalaxyScreenSpace, parameterName=Galaxy_PanSpeed, defaultValue=<rgba 0.005 0.015 0 0>}.RGB) {g=true, r=true}
  n168 = TextureSampleParameter2D(UVs=Add(A=Panner(Coordinate=@n166, Speed=@n167) {speedY=0.02}, B=NamedRerouteUsage())) {
   group=GalaxyScreenSpace
   parameterName=Galaxy
   texture=/Game/Characters/Player/Male/Medium/Bodies/M_Med_Soldier_04/Skins/Galaxy/Textures/T_FN_Nebula
  }
  n169 = ScalarParameter() {defaultValue=0.5, group=GalaxyScreenSpace, parameterName=Star Saturation}
  n170 = NamedRerouteUsage()
  n171 = TextureSample(UVs=Panner(Coordinate=NamedRerouteUsage()) {speedX=0.133, speedY=0.2}, Tex=NamedRerouteUsage()) {
   samplerSource=SSM_Wrap_WorldGroupSettings
   samplerType=SAMPLERTYPE_Masks
   texture=&T_Stars_03
  }
  n172 = LinearInterpolate(
   A=@n168.RGB
   B=Multiply(
    A=Desaturation(None=@n168.RGB, Fraction=@n169)
    B=VectorParameter() {group=GalaxyScreenSpace, parameterName=Star Brightness, defaultValue=<rgba 300 300 300 1>}.RGB
   )
   Alpha=LinearInterpolate(B=Multiply(A=@n170) {constB=0.02}, Alpha=Multiply(A=@n171.G, B=@n171.G)) {constB=0}
  )
  n173 = Multiply(
   A=VectorParameter() {group=GalaxyScreenSpace, parameterName=EdgeFresnel, defaultValue=<rgba 5.0909 8.586 10 1>}.RGB
   B=ScalarParameter() {defaultValue=6, group=BlackHole, parameterName=FresnelIntensityBlackHole}
  ) {constB=10}
  n174 = Multiply(A=@n173)
  n175 = LinearInterpolate(
   A=Multiply(
    A=Power(Base=@n172) {constExponent=1}
    B=ScalarParameter() {defaultValue=2, group=GalaxyScreenSpace, parameterName=GalaxyBackGroundBrightness}
   ) {constB=3}
   B=MaterialFunctionCall(Color_Morning=@n173, Color_Day=@n173, Color_Evening=@n174, Color_Night=@n174) {materialFunction=&Color_Time_of_Day}.Result
   Alpha=Fresnel(
    ExponentIn=ScalarParameter() {defaultValue=5, group=GalaxyScreenSpace, parameterName=Galaxy_FresEX}
    BaseReflectFractionIn=ScalarParameter() {defaultValue=0.04, group=GalaxyScreenSpace, parameterName=Galaxy_FresRfl}
   )
  )
  n176 = NamedRerouteUsage()
  NamedRerouteDeclaration(None=LinearInterpolate(
    A=MaterialFunctionCall(Color_Morning=@n163, Color_Day=@n163, Color_Evening=@n164, Color_Night=@n164) {materialFunction=&Color_Time_of_Day}.Result
    B=Multiply(A=@n165, B=@n175)
    Alpha=StaticComponentMaskParameter(None=@n176) {defaultG=true, group=BlackHole, parameterName=UseBlackHole}
   )) {name=CachedGalaxyEmissive, nodeColor=<rgba 0 1 0.3438 1>}
  n177 = MaterialFunctionCall() {materialFunction=&CameraDirectionVector}
  n178 = Multiply(A=Constant3Vector() {constant=<rgba 1 1 0 0>}, B=@n177.Result)
  n179 = Add(A=Multiply(B=Normalize(VectorInput=@n178)) {constA=0.25}, B=@n177.Result)
  n180 = Multiply(A=@n177.Result) {constB=50}
  n181 = Add(A=ActorPositionWS(), B=Constant3Vector() {constant=<rgba 0 0 90 1>})
  n182 = ScalarParameter() {defaultValue=128, group=GalaxyScreenSpace, parameterName=TextureResolution, sliderMax=2048}
  n183 = Multiply(A=@n182) {constB=0.5}
  n184 = InverseLinearInterpolate(
   A=Constant() {r=45}
   B=Constant() {r=90}
   Value=MaterialFunctionCall() {materialFunction=/Engine/Functions/Engine_MaterialFunctions02/Utility/FOV}.FOV angle in degrees
  )
  NamedRerouteDeclaration(None=MaterialFunctionCall(
    Plane Normal Axis=Normalize(VectorInput=@n179)
    Plane Center=Add(A=@n180, B=@n181)
    UV Scale=LinearInterpolate(A=@n183, B=@n182, Alpha=@n184)
   ) {materialFunction=&VirtualPlaneCoordinates}.UVs) {name=ScreenSpaceUVs, nodeColor=<rgba 0 0.7187 1 1>}
  n185 = MaterialFunctionCall(Float2=MaterialFunctionCall() {materialFunction=&ScreenResolution}.Visible Resolution) {materialFunction=&BreakOutFloat2Components}
  n186 = MaterialFunctionCall(Float2=Constant() {r=512}) {materialFunction=&BreakOutFloat2Components}
  n187 = Divide(B=Multiply(A=Divide(A=@n185.R, B=@n185.R), B=Divide(A=@n186.G, B=@n186.G) {desc=ratio}))
  n188 = MaterialFunctionCall(
   UVs=ScreenPosition().ViewportUV
   Texture Scale=AppendVector(
    A=MaterialFunctionCall(Float2=@n187) {materialFunction=&BreakOutFloat2Components}.R
    B=Constant() {r=1}
   )
  ) {materialFunction=&ScaleUVsByCenter}
  n189 = TextureSample(UVs=Divide(
    A=Panner(Coordinate=@n188.UVs) {speedX=0.01, speedY=-0.02}
    B=ScalarParameter() {defaultValue=4, group=GalaxyScreenSpace, parameterName=DistortionTiling, sliderMax=32}
   )) {samplerType=SAMPLERTYPE_Normal, texture=/Game/Gadgets/Assets/Telephone_Booth/Textures/T_BlindsNormalNoise_N}
  NamedRerouteDeclaration(None=Multiply(
    A=ComponentMask(None=@n189.RGB) {g=true, r=true}
    B=ScalarParameter() {defaultValue=0.08, group=GalaxyScreenSpace, parameterName=DistortionIntensity, sliderMax=1}
   ) {constB=0.02}) {name=DistortedUVs, nodeColor=<rgba 1 0.2812 0 1>}
  n190 = Multiply(A=Time(), B=ScalarParameter() {defaultValue=0.1, group=BlackHole, parameterName=CloudTimeScale}) {constB=0.1}
  n191 = Sine(None=@n190)
  NamedRerouteDeclaration(None=LinearInterpolate(Alpha=ConstantBiasScale(None=@n191)) {constA=1, constB=1.5}) {name=Sine01, nodeColor=<rgba 1 0 0.1641 1>}
  n192 = Sine(None=Add(A=@n190, B=Constant() {r=0.5}))
  NamedRerouteDeclaration(None=LinearInterpolate(Alpha=ConstantBiasScale(None=@n192)) {constA=1, constB=1.5}) {name=Sine02, nodeColor=<rgba 1 0 0.1641 1>}
  NamedRerouteDeclaration(None=@n131) {name=CachedNewLocalPos, nodeColor=<rgba 0 1 0.0391 1>}
  n193 = ScalarParameter() {defaultValue=1, group=BlackHole, parameterName=OriginalEmissiveIntensity}
  NamedRerouteDeclaration(None=StaticSwitchParameter(True=Constant(), False=Multiply(A=@n7.EmissiveColor, B=@n193)) {group=BlackHole, parameterName=ClearOriginalEmissive}) {name=MaterialEmissive, nodeColor=<rgba 0 0.2266 1 1>}
  n194 = Sine(None=Add(A=@n190, B=Constant() {r=1}))
  NamedRerouteDeclaration(None=LinearInterpolate(Alpha=ConstantBiasScale(None=@n194)) {constA=1, constB=1.5}) {name=Sine03, nodeColor=<rgba 1 0 0.1641 1>}
  NamedRerouteDeclaration(None=LinearInterpolate(A=@n14, Alpha=NamedRerouteUsage())) {name=CachedPolarSucking, nodeColor=<rgba 0.0781 0 1 1>}
  NamedRerouteDeclaration(None=LinearInterpolate(
    A=LinearInterpolate(Alpha=@n13.RGB) {constA=1, constB=10}
    B=ScalarParameter() {defaultValue=1.5, group=BlackHole, parameterName=InnerColorIntensityOverall}
    Alpha=NamedRerouteUsage()
   ) {constB=1.5}) {name=CachedPolarSucking_star, nodeColor=<rgba 0.0781 0 1 1>}
  NamedRerouteDeclaration(None=@n7.Metallic) {name=CachedMetalic, nodeColor=<rgba 1 0 0.6094 1>}
  n195 = Saturate(None=Multiply(A=@n119, B=Divide(A=@n118)))
  NamedRerouteDeclaration(None=Multiply(A=NamedRerouteDeclaration(None=@n195) {name=MetalicMask, nodeColor=<rgba 0.7578 0 1 1>}, B=@n125)) {name=CachedMetalicMask, nodeColor=<rgba 0.8281 1 0 1>}
  n196 = Distance(A=NamedRerouteUsage(), B=Constant() {r=0.3})
  n197 = Multiply(A=Saturate(None=OneMinus(None=@n196))) {constB=10}
  NamedRerouteDeclaration(None=Multiply(A=Saturate(None=@n197), B=NamedRerouteUsage())) {name=GradientRadialLinear_Inner, nodeColor=<rgba 0 0.3672 1 1>}
  n198 = PreSkinnedPosition()
  n199 = Divide(
   A=Add(A=@n198, B=VectorParameter() {group=BlackHole, parameterName=Chest Effect Inner Position Emissive, defaultValue=<rgba 5 2 -125 0>}.RGB)
   B=ScalarParameter() {defaultValue=8, group=BlackHole, parameterName=BlackHole_Inner_Radius}
  ) {constB=20}
  NamedRerouteDeclaration(None=ComponentMask(None=VertexInterpolator(VS=@n199).PS) {b=true, r=true}) {name=ChestCoord_Inner, nodeColor=<rgba 1 0 0.4453 1>}
  n200 = Divide(
   A=Add(A=@n198, B=VectorParameter() {group=BlackHole, parameterName=Chest Effect Position, defaultValue=<rgba 5 2 -125 0>}.RGB) {constB=0}
   B=ScalarParameter() {defaultValue=15, group=BlackHole, parameterName=BlackHole_Radius}
  ) {constB=20}
  NamedRerouteDeclaration(None=ComponentMask(None=VertexInterpolator(VS=@n200).PS) {b=true, r=true}) {name=ChestCoord, nodeColor=<rgba 1 0 0.4453 1>}
  n201 = Multiply(A=TextureSample(Tex=TextureObjectParameter() {parameterName=OverallMask, texture=/Game/Characters/Player/Female/Medium/Bodies/F_MED_SuperNovaTaro/Textures/T_SuperNova_Star_Mask}) {texture=&T_SuperNova_Star_Mask}.R) {constB=2.5}
  NamedRerouteDeclaration(None=LinearInterpolate(Alpha=Saturate(None=@n201)) {constA=1, constB=0}) {name=PaintStarMask, nodeColor=<rgba 0 1 0.1094 1>}
  NamedRerouteDeclaration(None=@n112.Result) {name=DebugCloud, nodeColor=<rgba 0.5469 0 1 1>}
  n202 = NamedRerouteUsage()
  n203 = TextureCoordinate()
  n204 = BumpOffset(Coordinate=@n203, Height=ScalarParameter() {defaultValue=-2, group=BlackHole, parameterName=BlackHole_SuckIntensity})
  n205 = Saturate(None=NamedRerouteUsage())
  n206 = ScalarParameter() {defaultValue=20, group=BlackHole, parameterName=BlackHoleSuckFrequency}
  n207 = Time()
  n208 = Frac(None=Subtract(A=Multiply(A=@n205, B=@n206), B=Frac(None=@n207)))
  n209 = Add(A=Sine(None=Divide(A=@n208)))
  n210 = NamedRerouteDeclaration(None=Multiply(A=Divide(A=@n209), B=NamedRerouteUsage())) {name=CachedDistortedAlpha, nodeColor=<rgba 1 0 0.1406 1>}
  n211 = TextureSample(UVs=NamedRerouteDeclaration(None=LinearInterpolate(A=@n203, B=@n204, Alpha=@n210)) {name=WavedUV, nodeColor=<rgba 1 0.8438 0 1>}) {
   samplerSource=SSM_Wrap_WorldGroupSettings
   texture=/Game/Characters/Player/Female/Medium/Bodies/F_MED_SuperNovaTaro/Textures/T_SuperNovaTaro_Body_D
  }
  n212 = Ceil(None=ComponentMask(None=NamedRerouteUsage()) {r=true})
  NamedRerouteDeclaration(None=StaticSwitchParameter(True=@n202, False=LinearInterpolate(A=@n211.RGB, B=@n202, Alpha=@n212)) {group=BlackHole, parameterName=UseCustomizedGalaxy}) {name=CachedBaseColorOUT, nodeColor=<rgba 1 0.2578 0 1>}
  n213 = Divide(A=Add(A=@n198, B=VectorParameter() {group=BlackHole, parameterName=Chest Effect Position Emissive, defaultValue=<rgba 5 2 -125 0>}.RGB) {constB=0}) {constB=10}
  NamedRerouteDeclaration(None=ComponentMask(None=VertexInterpolator(VS=@n213).PS) {b=true, r=true}) {name=CachedChestCoordEmissive, nodeColor=<rgba 0.5234 1 0 1>}
  n214 = Distance(A=NamedRerouteUsage(), B=Constant() {r=0.5})
  NamedRerouteDeclaration(None=Add(A=OneMinus(None=@n214), B=@n17)) {name=GradientRadialLinear_EM, nodeColor=<rgba 0 0.3672 1 1>}
  n215 = FunctionInput() {bUsePreviewValueAsDefault=true, inputName=TechArtMaskInput, sortPriority=2, previewValue=<1 1 1 1>}
  NamedRerouteDeclaration(None=@n215) {name=BlackHoleTechArtMask, nodeColor=<rgba 0 0.9531 1 1>}
  NamedRerouteUsage()
  NamedRerouteUsage()
  NamedRerouteUsage()
  NamedRerouteUsage()
  NamedRerouteUsage()
  NamedRerouteUsage()
  NamedRerouteUsage()
  NamedRerouteUsage()
  NamedRerouteUsage()
  NamedRerouteUsage()
  NamedRerouteUsage()
  NamedRerouteUsage()
  NamedRerouteUsage()
  NamedRerouteUsage()
  OneMinus(None=Saturate(None=Multiply(A=@n72) {constB=1.5}))
  Power(Base=@n133) {constExponent=20}
  n216 = NamedRerouteUsage()
  SmoothStep(Value=StaticComponentMaskParameter(None=@n216) {defaultG=true, group=BlackHole, parameterName=UseBlackHole}) {constMax=0.1}
  StaticComponentMaskParameter(None=@n216) {defaultR=true, group=BlackHole, parameterName=UseBlackHoleSolidPaint}
  n217 = Subtract(B=TextureCoordinate()) {constA=0.5, constB=0.5}
  n218 = MaterialFunctionCall(Float2=Multiply(A=@n217, B=@n217)) {materialFunction=&BreakOutFloat2Components}
  Subtract(B=SquareRoot(None=Add(A=@n218.R, B=@n218.R))) {constA=0.5}
  TextureCoordinate()
 ]}
