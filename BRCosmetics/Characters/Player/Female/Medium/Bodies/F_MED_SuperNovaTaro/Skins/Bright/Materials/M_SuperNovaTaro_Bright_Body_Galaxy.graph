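# Character body material for the Bright skin of the female SuperNova Taro, a Default Lit galaxy variant usable on skeletal meshes.
# It chains material functions through material attributes: character PreFX, CharacterSkinFX, a black hole FX function, character PostFX, then a cinematic post function.
# PreFX also feeds pre-skinned local position into the black hole effect.
# All shading logic lives in those shared and character-specific functions.
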Material /BRCosmetics/Characters/Player/Female/Medium/Bodies/F_MED_SuperNovaTaro/Skins/Bright/Materials/M_SuperNovaTaro_Bright_Body_Galaxy
graph {
 properties {
  bUseMaterialAttributes=true
  bUsedWithInstancedStaticMeshes=true
  bUsedWithMeshDeformer=true
  bUsedWithMorphTargets=true
  bUsedWithSkeletalMesh=true
  pixelDepthOffsetMode=PDOM_Legacy
  usedShadingModels=Default Lit
 }
 nodes [
  n1 = MaterialFunctionCall() {materialFunction=&BaseFN_CharacterFunction_PreFX}
  n2 = MaterialFunctionCall(
   PreskinnedPosition=@n1.Pre skinned Local Position
   Gmap=@n1.Pre skinned Local Position
   MaterialAttributes=@n1.Pre skinned Local Position
   AO=@n1.Pre skinned Local Position
  ) {materialFunction=&CharacterSkinFX}
  n3 = MaterialFunctionCall(
   AO=@n1.AO
   Distance Blend=@n1.AO
   Fresnel=@n1.AO
   Gmap=@n1.AO
   Pre Skinned Local Position=@n1.AO
   Material=MaterialFunctionCall!(MaterialInput=@n2.Result, InputWorldPos=@n1.Pre skinned Local Position, TechArtMaskInput=@n2.Result) {materialFunction=/Game/Characters/Player/Female/Medium/Bodies/F_MED_SuperNovaTaro/Materials/MF_BlackHole_FX_2023}.Result
  ) {materialFunction=&BaseFN_CharacterFunction_PostFX}
  MaterialFunctionCall(In=@n3.Result) {materialFunction=&CINE_MF_CharPostFX}
 ]
}
